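# Toon master material for Teriyaki accessory quest actors, defaulting to Ink Jar textures, that fakes cel lighting in emissive with black base color.
# Diffuse blended with CustomColor gets subtle sun-direction shading, masked highlights and Fresnel rim.
# Distance-field ink lines darken it, with thickness varying by camera distance and separate mobile parameters.
# Static switches add bob-and-rotate WPO and a camera-oriented emissive sweep; character PreFX and DeRez functions wrap the result.
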Material /BattlePassS18/QuestActors/Assets/Materials/M_TeriyakiToon_Accessories_Master
graph {
 properties {
  bUseMaterialAttributes=true
  bUsedWithMeshDeformer=true
  bUsedWithNiagaraMeshParticles=true
  bUsedWithSkeletalMesh=true
  pixelDepthOffsetMode=PDOM_Legacy
  shadingModel=MSM_Subsurface
  usedShadingModels=Subsurface
 }
 nodes [
  n1 = CollectionParameter() {
   collection=&FortniteMaterialParameters
   parameterName=SunlightDirectionalVector
  }
  ComponentMask(None=@n1) {b=true, g=true, r=true}
  Constant() {r=0.001}
  Constant()
  MaterialFunctionCall() {materialFunction=&ComponentPivotLocation}
  n2 = MaterialFunctionCall() {materialFunction=&BaseFN_CharacterFunction_PreFX}
  n3 = Constant()
  n4 = TextureSample() {samplerType=SAMPLERTYPE_Masks, texture=/BattlePassS18/QuestActors/Assets/Textures/T_Gadget_Ink_Jar_FX}
  n5 = DotProduct(A=Constant3Vector() {constant=<rgba 0 0 1 0>}, B=PixelNormalWS())
  n6 = Saturate(None=Add(A=@n4.R, B=Multiply(A=@n5) {constB=0.5}))
  n7 = MaterialFunctionCall(In=Fresnel(ExponentIn=Constant() {r=1.5}, Normal=PixelNormalWS()), Contrast=Constant() {r=5}) {materialFunction=&CheapContrast}
  n8 = Transform(None=CameraPositionWS()) {transformSourceType=TRANSFORMSOURCE_Camera, transformType=TRANSFORM_Local}
  n9 = PreSkinnedPosition()
  n10 = PreSkinnedLocalBounds()
  n11 = DotProduct(
   A=Multiply(A=Normalize(VectorInput=@n8), B=Constant3Vector() {constant=<rgba -1 -1 1 0>})
   B=Divide(A=VertexInterpolator(VS=@n9).PS, B=Subtract(A=@n10.Max, B=@n10.Max))
  )
  n12 = Constant() {r=0.15}
  n13 = Constant() {r=0.03}
  n14 = Constant() {r=1}
  n15 = Constant()
  n16 = Constant() {r=0.1}
  n17 = Constant() {r=1}
  n18 = Multiply(A=@n4.R, B=Add(A=SphereMask(A=@n11, B=@n12, Radius=@n13, Hardness=@n14), B=SphereMask(A=@n11, B=@n15, Radius=@n16, Hardness=@n17)))
  n19 = LinearInterpolate(
   A=TextureSample() {texture=/BattlePassS18/QuestActors/Assets/Textures/T_Gadget_Ink_Jar_D}.RGB
   B=VectorParameter() {parameterName=CustomColor, defaultValue=<rgba 0.9 0.9 0.9 0>}.RGB
   Alpha=TextureSample() {samplerType=SAMPLERTYPE_Masks, texture=&T_Gadget_Ink_Jar_FX}.R
  )
  n20 = CollectionParameter() {
   collection=&FortniteMaterialParameters
   parameterName=SunlightDirectionalVector
  }
  n21 = ComponentMask(None=@n20) {b=true, g=true, r=true}
  n22 = PixelNormalWS()
  n23 = Multiply(A=@n19, B=SmoothStep(Max=ScalarParameter() {defaultValue=0.1, parameterName=ShadingSmooth}, Value=DotProduct(A=@n21, B=@n22)) {constMax=0})
  n24 = Constant() {r=0.2}
  n25 = TextureSampleParameter2D() {parameterName=DistanceField_InkLines, texture=/BattlePassS18/QuestActors/Assets/Textures/T_Gadget_Ink_Jar_DF}
  n26 = TextureSampleParameter2D() {parameterName=DistanceField_InkLines_Mobile, texture=&T_Gadget_Ink_Jar_DF}
  n27 = ChannelMaskParameter(None=FeatureLevelSwitch(Default=@n25.RGB, ES2=@n26.RGB, ES3_1=@n26.RGB, SM5=@n25.RGB, SM6=@n25.RGB)) {parameterName=DistanceField_Channel}
  n28 = Multiply(
   A=Power(Base=@n27.R, Exp=ScalarParameter() {defaultValue=9, parameterName=DF_BlurPow}) {constExponent=3}
   B=ScalarParameter() {defaultValue=1, parameterName=DF_BlurMult}
  ) {constB=0.5}
  n29 = ScalarParameter() {defaultValue=0.05, group=NeonDepth, parameterName=DF_LineThickness_Near}
  n30 = ScalarParameter() {defaultValue=0.05, group=NeonDepth, parameterName=DF_LineThickness_Near_Mobile}
  n31 = FeatureLevelSwitch(Default=@n29, ES2=@n30, ES3_1=@n30, SM5=@n29, SM6=@n29)
  n32 = ScalarParameter() {defaultValue=0.15, group=NeonDepth, parameterName=DF_LineThickness_Far}
  n33 = ScalarParameter() {defaultValue=0.15, group=NeonDepth, parameterName=DF_LineThickness_Far_Mobile}
  n34 = FeatureLevelSwitch(Default=@n32, ES2=@n33, ES3_1=@n33, SM5=@n32, SM6=@n32)
  n35 = Distance(A=WorldPosition().XYZ, B=CameraPositionWS())
  n36 = ScalarParameter() {defaultValue=100, group=NeonDepth, parameterName=DF_FarDistance}
  n37 = ScalarParameter() {defaultValue=500, group=NeonDepth, parameterName=DF_NearDistance}
  n38 = Saturate(None=Divide(A=Subtract(A=@n35, B=@n36), B=Subtract(A=@n37, B=@n36)))
  n39 = Saturate(None=SphereMask(
    A=@n27.R
    B=Constant() {r=1}
    Radius=LinearInterpolate(A=@n31, B=@n34, Alpha=@n38)
    Hardness=ScalarParameter() {defaultValue=0.8, parameterName=DF_LineHardness}
   ) {attenuationRadius=0.1})
  n40 = LinearInterpolate(A=Constant3Vector(), B=Constant3Vector() {constant=<rgba 1 1 1 0>}, Alpha=Multiply(A=OneMinus(None=@n28), B=OneMinus(None=@n39)))
  n41 = Add(A=Add(A=Multiply(A=@n6, B=@n7.Result), B=Multiply(A=@n18) {constB=0.1}), B=Multiply(A=LinearInterpolate(A=@n19, B=@n23, Alpha=@n24), B=@n40))
  n42 = SetMaterialAttributes!(
   MaterialAttributes=@n2.Result
   Base Color=@n3
   Metallic=@n3
   Emissive Color=Add(A=Multiply(A=@n41) {constB=0.6}, B=NamedRerouteUsage())
   Specular=@n3
   Roughness=VertexColor().R
   Subsurface Color=Constant() {r=0.01}
   Ambient Occlusion=NamedRerouteUsage()
   World Position Offset=NamedRerouteUsage()
   Normal=TextureSampleParameter2D() {parameterName=Normals, samplerType=SAMPLERTYPE_Normal, texture=/BattlePassS18/QuestActors/Assets/Textures/T_Gadget_Ink_Jar_N}.RGB
  ) {attributeSetTypes [
    #1
    #2
    #3
    #4
    #5
    #6
    #7
    #8
    #9
   ]}
  MaterialFunctionCall!(Fresnel=@n2.Fresnel, Material=@n42) {materialFunction=/Game/Packages/Fortress_SharedMaterials/Base_Material_Functions/CharacterDeRez_Transition}
  n43 = Multiply(A=Time(), B=ScalarParameter() {defaultValue=0.4, parameterName=WPOTimeScale})
  n44 = LinearInterpolate(
   A=VectorParameter() {parameterName=BobHeightMax, defaultValue=<rgba 0 0 10 0>}.RGB
   B=VectorParameter() {parameterName=BobHeightMin, defaultValue=<rgba 0 0 -2 0>}.RGB
   Alpha=Sine(None=@n43)
  )
  n45 = ScalarParameter() {defaultValue=-0.01, parameterName=Rotation Angle Min}
  n46 = ScalarParameter() {defaultValue=0.01, parameterName=Rotation Angle Max}
  n47 = Cosine(None=@n43)
  n48 = VectorParameter() {parameterName=WorldPostionOfBlackHole, defaultValue=<rgba -30000 25536 1952 0>}
  n49 = RotateAboutAxis(
   NormalizedRotationAxis=Constant3Vector() {constant=<rgba 0 0 1 0>}
   RotationAngle=Multiply(A=LinearInterpolate(A=@n45, B=@n46, Alpha=@n47), B=OneMinus(None=@n48.A))
   PivotPoint=ObjectPositionWS()
   Position=WorldPosition().XYZ
  )
  NamedRerouteDeclaration(None=StaticSwitchParameter(True=Add(A=@n44, B=@n49), False=Constant()) {parameterName=Use WPO}) {name=WPO, nodeColor=<rgba 0.3359 0 1 1>}
  n50 = Multiply(
   A=Subtract(A=WorldPosition().XYZ, B=ObjectPositionWS())
   B=ScalarParameter() {defaultValue=0.001, group=Emissive Sweep, parameterName=WorldTileScale}
  )
  n51 = Normalize(VectorInput=Transform(None=VectorParameter() {group=Emissive Sweep, parameterName=SweepDirectionCamSpace, defaultValue=<rgba -0.5 0.45 -1 0>}.RGB) {transformSourceType=TRANSFORMSOURCE_Camera})
  n52 = ComponentMask(None=Multiply(A=DotProduct(A=@n50, B=@n51), B=@n51)) {b=true, r=true}
  n53 = Constant() {r=0.28}
  n54 = Frac(None=ComponentMask(None=Panner(Coordinate=@n52, Speed=@n53) {speedX=0.2, speedY=0.2}) {g=true})
  n55 = Add(A=@n54) {constB=0.1}
  n56 = OneMinus(None=Multiply(A=Subtract(A=@n54) {constB=0.9}) {constB=10})
  n57 = MaterialFunctionCall() {materialFunction=&MF_Fresnel_FX}
  n58 = Multiply(A=Multiply(A=Floor(None=@n55), B=Power(Base=@n56) {constExponent=3}), B=Saturate(None=Add(A=@n57.Fresnel Inverted) {constB=0.1}))
  n59 = VectorParameter() {group=Emissive Sweep, parameterName=SweepColour, defaultValue=<rgba 1.8958 2 1.9078 0>}
  n60 = LinearInterpolate(
   A=@n59.RGB
   B=Multiply(A=@n59.RGB, B=ScalarParameter() {defaultValue=15, group=Emissive Sweep, parameterName=FBoostScale})
   Alpha=VertexColor().G
  )
  NamedRerouteDeclaration(None=StaticSwitchParameter(True=Multiply(A=@n58, B=@n60), False=Constant()) {group=Emissive Sweep, parameterName=UseSweep}) {name=EmissiveSweep, nodeColor=<rgba 0 1 0.25 1>}
  NamedRerouteDeclaration(None=@n40) {name=DF Line Preview, nodeColor=<rgba 0 1 0.0625 1>}
  ScalarParameter() {defaultValue=0.15, parameterName=lineThickness}
 ]
}
