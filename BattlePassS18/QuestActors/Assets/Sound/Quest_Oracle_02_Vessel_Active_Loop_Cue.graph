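# Sound cue for the active state of the Oracle vessel quest actor in Battle Pass S18, layering a start sound with a sustained loop.
# A mixer plays the start wave once alongside the looping wave, weighting the start louder.
# An enveloper fades the loop in over 1.5 seconds while a modulator holds it at raised pitch.
# It uses the quest sound class, vessel attenuation and a concurrency override limiting it to two instances.
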SoundCue /BattlePassS18/QuestActors/Assets/Sound/Quest_Oracle_02_Vessel_Active_Loop_Cue
root=Mixer_0
nodes [
 {
  id=Enveloper_0
  children [WavePlayer_4]
  properties {pitchCurve {editorCurveData {keys [{time=0}]}}, volumeCurve {editorCurveData {keys [<at 0 0 RCIM_Cubic>,<at 1.5 1>]}}}
 }
 {id=Mixer_0, children [Modulator_0,Modulator_1], properties {inputVolume [1.35,0.85]}}
 {id=Modulator_0, children [WavePlayer_0], properties {pitchMax=1, pitchMin=1, volumeMax=1, volumeMin=1}}
 {id=Modulator_1, children [Enveloper_0], properties {pitchMax=1.1, pitchMin=1.1, volumeMax=1, volumeMin=1}}
 {id=WavePlayer_0, properties {soundWaveAssetPtr=&/Quest_Oracle_02_Vessel_Loop_Start}}
 {id=WavePlayer_4, properties {bLooping=true, soundWaveAssetPtr=&/Quest_Oracle_02_Vessel_Active_Loop}}
]
properties {
 attenuationSettings=&/Quest_Oracle_02_Vessel_Att
 bOverrideConcurrency=true
 maxDistance=2050
 soundClassObject=&Fort_Quests
 volumeMultiplier=4
 concurrencyOverrides {maxCount=2}
}
