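# Sound cue for the Battle Pass S18 Oracle quest 02 vessel, named as its destroy sound.
# A single wave player feeds the output directly and plays the vessel's Interact_01 wave, with no randomization or modulation.
# The cue uses vessel-specific attenuation, the Fort_Quests sound class and a boosted volume, and overrides concurrency to cap simultaneous instances.
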SoundCue /BattlePassS18/QuestActors/Assets/Sound/Quest_Oracle_02_Vessel_Destroy_Cue
root=WavePlayer_4
nodes [{id=WavePlayer_4, properties {soundWaveAssetPtr=&/Quest_Oracle_02_Vessel_Interact_01}}]
properties {
 attenuationSettings=&/Quest_Oracle_02_Vessel_Att
 bOverrideConcurrency=true
 maxDistance=2050
 soundClassObject=&Fort_Quests
 volumeMultiplier=1.5
 concurrencyOverrides {maxCount=3}
}
