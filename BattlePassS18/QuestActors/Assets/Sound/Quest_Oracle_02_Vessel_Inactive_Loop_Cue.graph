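# Sound Cue for the inactive-state loop of the Oracle 02 quest vessel in Battle Pass S18, routed to the quests sound class.
# A Mixer layers a quieter, pitched-down start one-shot with the looping inactive wave.
# An Enveloper fades the loop in from silence, so the start sound covers its onset.
# Playback uses the vessel's attenuation asset and overrides concurrency to cap simultaneous instances at two.
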SoundCue /BattlePassS18/QuestActors/Assets/Sound/Quest_Oracle_02_Vessel_Inactive_Loop_Cue
root=Mixer_0
nodes [
 {
  id=Enveloper_0
  children [WavePlayer_4]
  properties {pitchCurve {editorCurveData {keys [{time=0}]}}, volumeCurve {editorCurveData {keys [<at 0 0>,<at 1 1>]}}}
 }
 {id=Mixer_0, children [Modulator_0,Enveloper_0], properties {inputVolume [0.7,1]}}
 {id=Modulator_0, children [WavePlayer_0], properties {pitchMax=0.8, pitchMin=0.8, volumeMax=1, volumeMin=1}}
 {id=WavePlayer_0, properties {soundWaveAssetPtr=&/Quest_Oracle_02_Vessel_Loop_Start}}
 {id=WavePlayer_4, properties {bLooping=true, soundWaveAssetPtr=&/Quest_Oracle_02_Vessel_Inactive_Loop}}
]
properties {
 attenuationSettings=&/Quest_Oracle_02_Vessel_Att
 bOverrideConcurrency=true
 maxDistance=2050
 soundClassObject=&Fort_Quests
 volumeMultiplier=0.9
 concurrencyOverrides {maxCount=2}
}
