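# Sound cue for the looping backpack audio in the Battle Pass S18 Oracle quest 03, using Vessel sounds in the Fort_Quests sound class.
# A weighted mixer layers a louder one-shot start wave with a looping wave.
# An Enveloper shapes only the loop, fading it in quickly, slowly lowering its volume and shifting its pitch over time.
# It borrows the Oracle 02 backpack attenuation and overrides concurrency to allow two instances.
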SoundCue /BattlePassS18/QuestActors/Assets/Sound/Quest_Oracle_03_Backpack_Loop_Cue
root=Mixer_0
nodes [
 {
  id=Enveloper_0
  children [WavePlayer_4]
  properties {
   pitchCurve {editorCurveData {keys [<at 0 0.6>, {time=10}]}}
   volumeCurve {editorCurveData {keys [{arriveTangent=2.8, interpMode=RCIM_Cubic, leaveTangent=2.8, tangentMode=RCTM_User, time=0, value=0}, <at 1.5 1>, <at 20 0.15>]}}
  }
 }
 {id=Mixer_0, children [WavePlayer_0,Enveloper_0], properties {inputVolume [2.5,1.5]}}
 {id=WavePlayer_0, properties {soundWaveAssetPtr=&/Quest_Oracle_03_Vessel_Loop_Start}}
 {id=WavePlayer_4, properties {bLooping=true, soundWaveAssetPtr=&/Quest_Oracle_03_Vessel_Loop}}
]
properties {
 attenuationSettings=&/Quest_Oracle_02_Vessel_Backpack_Att
 bOverrideConcurrency=true
 maxDistance=4300
 soundClassObject=&Fort_Quests
 volumeMultiplier=1
 concurrencyOverrides {maxCount=2}
}
