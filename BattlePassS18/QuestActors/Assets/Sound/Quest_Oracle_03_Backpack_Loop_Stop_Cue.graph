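# Sound cue for the stop sound of the Oracle 03 quest backpack loop in the Battle Pass S18 quest actors.
# A single wave player feeds the Vessel loop-stop wave straight to the output, with no randomization or mixing.
# It uses the quest sound class, borrows the Oracle 02 Vessel backpack attenuation, and overrides concurrency to cap simultaneous instances.
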SoundCue /BattlePassS18/QuestActors/Assets/Sound/Quest_Oracle_03_Backpack_Loop_Stop_Cue
root=WavePlayer_1
nodes [{id=WavePlayer_1, properties {soundWaveAssetPtr=&/Quest_Oracle_03_Vessel_Loop_Stop}}]
properties {
 attenuationSettings=&/Quest_Oracle_02_Vessel_Backpack_Att
 bOverrideConcurrency=true
 maxDistance=4300
 soundClassObject=&Fort_Quests
 volumeMultiplier=1
 concurrencyOverrides {maxCount=2}
}
