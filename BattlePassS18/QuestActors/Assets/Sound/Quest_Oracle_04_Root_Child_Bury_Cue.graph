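# Sound cue for burying a root child in the Battle Pass Season 18 Oracle quest, stage 04.
# A random node picks one of four equally weighted Destroy wave variations, reused here for the bury sound, and feeds a modulator that holds volume fixed.
# It plays in the quest sound class with the quest root's attenuation asset and a concurrency override limiting simultaneous instances.
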SoundCue /BattlePassS18/QuestActors/Assets/Sound/Quest_Oracle_04_Root_Child_Bury_Cue
root=Modulator_0
nodes [
 {id=Modulator_0, children [Random_1], properties {volumeMax=1, volumeMin=1}}
 {
  id=Random_1
  children [WavePlayer_4,WavePlayer_5,WavePlayer_6,WavePlayer_7]
  properties {bShouldExcludeFromBranchCulling=true, weights [1,1,1,1]}
 }
 {id=WavePlayer_4, properties {soundWaveAssetPtr=&/Quest_Oracle_04_Root_Child_Destroy_01}}
 {id=WavePlayer_5, properties {soundWaveAssetPtr=&/Quest_Oracle_04_Root_Child_Destroy_02}}
 {id=WavePlayer_6, properties {soundWaveAssetPtr=&/Quest_Oracle_04_Root_Child_Destroy_03}}
 {id=WavePlayer_7, properties {soundWaveAssetPtr=&/Quest_Oracle_04_Root_Child_Destroy_04}}
]
properties {
 attenuationSettings=&/Quest_Oracle_04_Root_Att
 bOverrideConcurrency=true
 maxDistance=8200
 soundClassObject=&Fort_Quests
 volumeMultiplier=0.9
 concurrencyOverrides {maxCount=3}
}
