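# Sound cue named for a hit on the child root in Oracle quest 04 of Battle Pass S18, routed through the Fort_Quests sound class.
# Its only node is a root Modulator with fixed volume and no child, so no sound wave is connected and the cue has nothing to play.
# It still sets a shared root attenuation asset, a slightly reduced volume and a concurrency override capping simultaneous instances.
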SoundCue /BattlePassS18/QuestActors/Assets/Sound/Quest_Oracle_04_Root_Child_Hit_Cue
root=Modulator_0
nodes [{id=Modulator_0, children [null], properties {volumeMax=1, volumeMin=1}}]
properties {
 attenuationSettings=&/Quest_Oracle_04_Root_Att
 bOverrideConcurrency=true
 maxDistance=8200
 soundClassObject=&Fort_Quests
 volumeMultiplier=0.9
 concurrencyOverrides {maxCount=3}
}
