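# Sound Cue for the looping idle sound of the Oracle quest's Root Child actor in Battle Pass S18.
# A looping Wave Player feeds a Modulator with volume variation pinned off and pitch range at default, then an Enveloper whose volume curve rises from silence to full.
# The cue plays at reduced volume in the Fort_Quests sound class, uses the Root Idle attenuation asset, and overrides concurrency to cap simultaneous instances.
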SoundCue /BattlePassS18/QuestActors/Assets/Sound/Quest_Oracle_04_Root_Child_Idle_Loop_Cue
root=Enveloper_0
nodes [
 {
  id=Enveloper_0
  children [Modulator_0]
  properties {pitchCurve {editorCurveData {keys [{time=0}]}}, volumeCurve {editorCurveData {keys [<at 0 0>,<at 1 1>]}}}
 }
 {id=Modulator_0, children [WavePlayer_4], properties {volumeMax=1, volumeMin=1}}
 {id=WavePlayer_4, properties {bLooping=true, soundWaveAssetPtr=&/Quest_Oracle_04_Root_Child_Idle_Loop}}
]
properties {
 attenuationSettings=&/Quest_Oracle_04_Root_Idle_Att
 bOverrideConcurrency=true
 maxDistance=1000
 soundClassObject=&Fort_Quests
 volumeMultiplier=0.4
 concurrencyOverrides {maxCount=3}
}
