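# Sound cue for the destruction of the main root in the Oracle quest 04 of the Season 18 Battle Pass.
# A Random node picks one of four equally weighted destroy waves and is excluded from branch culling.
# A Modulator at the root holds volume fixed, leaving only its default pitch variation.
# The cue uses the quest root attenuation and Fort_Quests sound class, and overrides concurrency to cap simultaneous plays.
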SoundCue /BattlePassS18/QuestActors/Assets/Sound/Quest_Oracle_04_Root_Main_Destroy_Cue
root=Modulator_0
nodes [
 {id=Modulator_0, children [Random_0], properties {volumeMax=1, volumeMin=1}}
 {
  id=Random_0
  children [WavePlayer_0,WavePlayer_1,WavePlayer_2,WavePlayer_3]
  properties {bShouldExcludeFromBranchCulling=true, weights [1,1,1,1]}
 }
 {id=WavePlayer_0, properties {soundWaveAssetPtr=&/Quest_Oracle_04_Root_Main_Destroy_01}}
 {id=WavePlayer_1, properties {soundWaveAssetPtr=&/Quest_Oracle_04_Root_Main_Destroy_02}}
 {id=WavePlayer_2, properties {soundWaveAssetPtr=&/Quest_Oracle_04_Root_Main_Destroy_03}}
 {id=WavePlayer_3, properties {soundWaveAssetPtr=&/Quest_Oracle_04_Root_Main_Destroy_04}}
]
properties {
 attenuationSettings=&/Quest_Oracle_04_Root_Att
 bOverrideConcurrency=true
 maxDistance=8200
 soundClassObject=&Fort_Quests
 volumeMultiplier=0.9
 concurrencyOverrides {maxCount=3}
}
