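# Sound cue for the Quest Oracle 04 root actor in Battle Pass S18, layering a grow sound with an idle loop.
# A mixer plays the one-shot grow wave, boosted well above unity, alongside the quieter looping idle wave.
# An enveloper fades the loop in over the first second and applies a two-second pitch curve.
# It uses shared root attenuation, the quest sound class and overridden concurrency limited to three instances.
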SoundCue /BattlePassS18/QuestActors/Assets/Sound/Quest_Oracle_04_Root_Main_Idle_Loop_Cue
root=Mixer_0
nodes [
 {
  id=Enveloper_0
  children [WavePlayer_4]
  properties {
   pitchCurve {editorCurveData {keys [<at 0 0.85>, {time=2}]}}
   volumeCurve {editorCurveData {keys [<at 0 0 RCIM_Cubic>, {arriveTangent=2.6105, interpMode=RCIM_Cubic, leaveTangent=2.6105, tangentMode=RCTM_User, time=1, value=1}]}}
  }
 }
 {id=Mixer_0, children [WavePlayer_0,Enveloper_0], properties {inputVolume [4,0.75]}}
 {id=WavePlayer_0, properties {soundWaveAssetPtr=&/Quest_Oracle_04_Root_Main_Grow_01}}
 {id=WavePlayer_4, properties {bLooping=true, soundWaveAssetPtr=&/Quest_Oracle_04_Root_Main_Idle_Loop}}
]
properties {
 attenuationSettings=&/Quest_Oracle_04_Root_Att
 bOverrideConcurrency=true
 maxDistance=8200
 soundClassObject=&Fort_Quests
 volumeMultiplier=1
 concurrencyOverrides {maxCount=3}
}
