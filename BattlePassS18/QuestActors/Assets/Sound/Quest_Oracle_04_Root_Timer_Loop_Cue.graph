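# Sound cue for the root timer loop in the Battle Pass S18 Oracle quest 04, routed to the Fort_Quests sound class.
# A mixer layers a one-shot start wave, slightly pitched down by a fixed modulator and mixed at reduced volume, with a looping timer wave.
# An enveloper ramps the loop up from silence, so it fades in beneath the start sound.
# Concurrency is overridden to cap simultaneous instances.
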SoundCue /BattlePassS18/QuestActors/Assets/Sound/Quest_Oracle_04_Root_Timer_Loop_Cue
root=Mixer_0
nodes [
 {
  id=Enveloper_0
  children [WavePlayer_1]
  properties {pitchCurve {editorCurveData {keys [{time=0}]}}, volumeCurve {editorCurveData {keys [<at 0 0>,<at 1 1>]}}}
 }
 {id=Mixer_0, children [Modulator_1,Enveloper_0], properties {inputVolume [0.75,1]}}
 {id=Modulator_1, children [WavePlayer_0], properties {pitchMax=0.9, pitchMin=0.9, volumeMax=1, volumeMin=1}}
 {id=WavePlayer_0, properties {soundWaveAssetPtr=&/Quest_Oracle_04_Root_Timer_Loop_Start}}
 {id=WavePlayer_1, properties {bLooping=true, soundWaveAssetPtr=&/Quest_Oracle_04_Root_Timer_Loop}}
]
properties {
 bOverrideConcurrency=true
 maxDistance=2097152
 soundClassObject=&Fort_Quests
 volumeMultiplier=0.5
 concurrencyOverrides {maxCount=3}
}
