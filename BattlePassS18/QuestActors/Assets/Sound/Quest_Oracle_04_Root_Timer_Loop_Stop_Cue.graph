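# Sound cue for the stop of the root timer loop in the Season 18 Oracle quest 04, classed as a Fort quest sound.
# A single wave player plays the loop-stop wave through a modulator that pins it to a fixed, lowered pitch at unchanged volume.
# The cue halves overall volume, carries a very large max distance, and overrides concurrency to cap simultaneous instances at a few.
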SoundCue /BattlePassS18/QuestActors/Assets/Sound/Quest_Oracle_04_Root_Timer_Loop_Stop_Cue
root=Modulator_1
nodes [
 {id=Modulator_1, children [WavePlayer_0], properties {pitchMax=0.65, pitchMin=0.65, volumeMax=1, volumeMin=1}}
 {id=WavePlayer_0, properties {soundWaveAssetPtr=&/Quest_Oracle_04_Root_Timer_Loop_Stop}}
]
properties {
 bOverrideConcurrency=true
 maxDistance=2097152
 soundClassObject=&Fort_Quests
 volumeMultiplier=0.5
 concurrencyOverrides {maxCount=3}
}
